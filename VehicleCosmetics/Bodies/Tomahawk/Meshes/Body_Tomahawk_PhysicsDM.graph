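# Physics asset for the Tomahawk vehicle body cosmetic, giving its mesh a simplified physics shape.
# A single body on the root bone holds one box, longest along X and offset forward and up from the bone origin.
# It has no constraints or other bodies, so the whole vehicle body is represented as one rigid volume with no articulated parts.
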PhysicsAsset /VehicleCosmetics/Bodies/Tomahawk/Meshes/Body_Tomahawk_PhysicsDM
bodies {root [{shapeType=Box, extentX=444, extentY=216, extentZ=116, shapeName=root_box, center=<30 0 88.5>}]}
constraints []
